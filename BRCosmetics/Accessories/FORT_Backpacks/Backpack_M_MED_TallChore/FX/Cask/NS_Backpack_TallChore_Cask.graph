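# Niagara system for the TallChore Cask back bling cosmetic, drawing glowing rings with drifting wisps.
# NE_Backpack_DrapeUltra partitions two burst-spawned particles into differently scaled, layered meshes, with waveform-pulsed purple-to-white tints that also color attached lights.
# WhispFlames continuously spawns two groups around a ring, masking velocity, lifetime and color per group and tagging each for renderer visibility.
# Glow and Decal add persistent sprites tinted by system-level ring colors.
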NiagaraSystem /BRCosmetics/Accessories/FORT_Backpacks/Backpack_M_MED_TallChore/FX/Cask/NS_Backpack_TallChore_Cask
dependencies {
 usedDataInterfaces [&NiagaraDataInterfaceCurve, &NiagaraDataInterfaceColorCurve]
 usedDynamicInputs []
 usedModules [
  &SystemState
  &self:SystemSpawnScript.SystemScriptSource.SystemScriptGraph.NiagaraNodeAssignment_0.SetVariables_#1
 ]
}
emitters {
 Decal {
  emitterUpdateScript [
   &EmitterState(Life Cycle Mode=::NewEnumerator1, Loop Behavior=::NewEnumerator1, Loop Duration=1, UseLoopDelay=false)
   &SpawnBurst_Instantaneous(Use Spawn Probability=false, UseLoopCountLimit=false)
  ]
  particleSpawnScript [
   &InitializeParticle(
    Color=$System.RingWhispColor
    Position Offset=<0 -6.2463 -1.6951>
    Sprite Size=<20 20>
    Sprite Size Mode=::NewEnumerator4
    UsePositionOffset=true
    Write Material Random=true
   )
   &CameraOffset(Camera Offset Amount=20)
  ]
  particleUpdateScript [
   &SolveForcesAndVelocity(Clamp Velocity=false, Limit Acceleration=false)
   &ParticleState(Kill Particles When Lifetime Has Elapsed=false)
   &DynamicMaterialParameters(
    Index 0 Param 1=2
    Index 0 Param 3=1
    Module.Index0Float0Write=true
    Module.Index0Float1Write=true
    Module.Index0Float2Write=true
    Module.Index0Float3Write=true
   )
  ]
  rendererClasses [&NiagaraSpriteRendererProperties]
  renderers [{rendererClass=&NiagaraSpriteRendererProperties}]
 }
 Glow {
  emitterUpdateScript [
   &EmitterState(Life Cycle Mode=::NewEnumerator1, Loop Behavior=::NewEnumerator1, Loop Duration=1, UseLoopDelay=false)
   &SpawnBurst_Instantaneous(Spawn Count=2, Use Spawn Probability=false, UseLoopCountLimit=false)
  ]
  particleSpawnScript [&InitializeParticle(
    Color=$System.RingInnerColor
    Position Offset=<0 -6.2463 -1.6951>
    Sprite Size=/Niagara/DynamicInputs/Helpers/MaskVector2d(
     Mask Source:NiagaraInt32=1
     Mask Parameter:NiagaraInt32=&ReturnExecIndex()
     Mask:NiagaraInt32=0
     Masked Vector2D:Vector2f=<31 31>
     Passthrough Vector2D:Vector2f=<20 30>
    )
    Sprite Size Mode=::NewEnumerator4
    UsePositionOffset=true
    Write Material Random=true
   )]
  particleUpdateScript [
   &SolveForcesAndVelocity(Clamp Velocity=false, Limit Acceleration=false)
   &ParticleState(Kill Particles When Lifetime Has Elapsed=false)
   &AlignSpriteToMeshOrientation(
    Mesh Orientation Relative Sprite Alignment Vector=&MaskVector(
     Mask Source=1
     Mask Parameter=&ReturnExecIndex()
     Mask=0
     Masked Vector=<0 0 1>
     Passthrough Vector=<0 0.5 1>
    )
    Mesh Orientation Relative Sprite Facing Vector:Vector3f=&MaskVector(
     Mask Source=1
     Mask Parameter=&ReturnExecIndex()
     Mask=0
     Masked Vector=<0 -1 0.4907>
     Passthrough Vector=<-1 0 0>
    )
    SetAlignment=true
    SetFacing=true
   )
  ]
  rendererClasses [&NiagaraSpriteRendererProperties]
  renderers [{rendererClass=&NiagaraSpriteRendererProperties}]
 }
 NE_Backpack_DrapeUltra {
  emitterUpdateScript [
   &EmitterState(
    Inactive Response=::NewEnumerator1
    Life Cycle Mode=::NewEnumerator1
    Loop Behavior=::NewEnumerator1
    Loop Duration Mode=::NewEnumerator1
    UseLoopDelay=false
   )
   &SpawnBurst_Instantaneous(Spawn Count=2, Use Spawn Probability=false, UseLoopCountLimit=false)
  ]
  particleSpawnScript [
   &PartitionParticles()
   &InitializeParticle(
    Lifetime=0
    Mesh Scale Mode=::NewEnumerator3
    Mesh Uniform Scale=&MergeHelper_FloatSwitch(0=0.7, 1=0.4, 2=?, 3=?, Index=$Particles.PartitionParticles.Partition, Num=::NewEnumerator0)
    Position Offset=&MergeHelper_VectorSwitch(0=<0 -1.3901 -1.2025>, 1=<0 -5.5 -0.3796>, 2=?, 3=?, Index=$Particles.PartitionParticles.Partition, Num=::NewEnumerator0)
    Sprite Size Mode=::NewEnumerator3
    Uniform Sprite Size=25
    UsePositionOffset=true
    Write Material Random=true
   )
   /SetVariables(Particles.VisibilityTag=$Particles.PartitionParticles.Partition) {isSetParameters=true}
   &InitialMeshOrientation(
    Mesh Axis To Orient=<0 0 0>
    Mesh Orientation Mode=::NewEnumerator2
    Rotation=&MergeHelper_VectorSwitch(0=<0 0 0>, 1=<0.19 0 0>, 2=?, 3=?, Index=$Particles.PartitionParticles.Partition, Num=::NewEnumerator0)
    Use Rotation Vector=true
   )
  ]
  particleUpdateScript [
   &ParticleState(Kill Particles When Lifetime Has Elapsed=false, Loop Particles Lifetime=true)
   &ScaleColor(
    Scale RGB=&MergeHelper_VectorSwitch(
     0=&Lerp_Vector(
      A=&MakeVectorFromLinearColorRGB(Color=<rgba 0.9 0.5733 0.8183 1>)
      B=&MakeVectorFromLinearColorRGB(Color=<rgba 1.07 1.07 0.9095 1>)
      Alpha=&Waveform("Waveform [1]"=::NewEnumerator12, "[1] Frequency"=1.3, "[1] Amplitude Min/Max"=<0 1>)
     )
     1=&Lerp_Vector(
      A=&MakeVectorFromLinearColorRGB(Color=<rgba 0.7674 0.4038 0.85 1>)
      B=&MakeVectorFromLinearColorRGB(Color=<rgba 1.05 1.05 0.8925 1>)
      Alpha=&Waveform("Waveform [1]"=::NewEnumerator12, "[1] Frequency"=1.4, "[1] Amplitude Min/Max"=<0 1>)
     )
     2=?
     3=?
     Index=$Particles.PartitionParticles.Partition
     Num=::NewEnumerator0
    )
    ScaleA=true
    ScaleRGB=true
   )
   &FN_Light_Attributes(
    Color=&MultiplyLinearColor(
     Linear Color=<rgba 0.8179 0.1347 1 1>
     Scale Factor=&MakeVector4_FromLinearColor(LinearColor=$Particles.Color)
    )
    Create Color Attribute=true
    Create Exponent Attribute=true
    Create Position Attribute=true
    Create Radius Attribute=true
    Enable Particle Light Enable=true
    Exponent=2
    Position=&AddVectorToPosition(Position=$Particles.Position, Vector=<0 -15 0>)
    Radius=32
    Scale Color=2
   )
   /SetVariables(
    Particles.DetailRing:Vector3f=&Multiply_VectorByFloat(Vector=$Particles.Initial.Scale, Float=0.9524)
    Particles.InnerRingSize:Vector3f=&Multiply_VectorByFloat(Vector=$Particles.Initial.Scale, Float=0.9)
   ) {isSetParameters=true}
  ]
  rendererClasses [&NiagaraMeshRendererProperties, &NiagaraLightRendererProperties]
  renderers [
   {rendererClass=&NiagaraMeshRendererProperties, rendererIndex=0}
   {rendererClass=&NiagaraMeshRendererProperties, rendererIndex=1}
   {rendererClass=&NiagaraLightRendererProperties, rendererIndex=2}
   {rendererClass=&NiagaraMeshRendererProperties, rendererIndex=3}
   {rendererClass=&NiagaraMeshRendererProperties, rendererIndex=4}
   {rendererClass=&NiagaraMeshRendererProperties, rendererIndex=5}
   {rendererClass=&NiagaraMeshRendererProperties, rendererIndex=6}
  ]
 }
 WhispFlames {
  emitterUpdateScript [
   &EmitterState()
   &SpawnRate(SpawnRate=3.5, Use Spawn Probability=false)
   &SpawnRate(SpawnGroup=1, SpawnRate=2, Use Spawn Probability=false) {module=SpawnRate001}
  ]
  particleSpawnScript [
   /SetVariables(Particles.VisibilityTag=&SpawnGroup()) {isSetParameters=true}
   &InitializeParticle(
    Color Mode=::NewEnumerator0
    Lifetime=&MaskFloatBySpawnGroup(
     Masked Float=&RandomRangeFloat(Minimum=1.2, Maximum=2)
     Passthrough Float=&RandomRangeFloat(Minimum=1, Maximum=1.5)
    )
    Mesh Scale Max=<0.8 0.8 0.55>
    Mesh Scale Min=<0.7 0.7 0.3>
    Mesh Scale Mode=::NewEnumerator2
    Sprite Rotation Mode=::NewEnumerator2
    Sprite Size Mode=::NewEnumerator3
    Uniform Sprite Size=4
    UsePositionOffset=false
    Write Material Random=true
   )
   &ShapeLocation(
    Disc Coverage=0.6587
    MaskSpawnGroup=false
    Offset=<0 -5 -1>
    Offset Mode=::NewEnumerator0
    Override Seed=false
    Ring Radius=9.4
    Rotation Mode=::NewEnumerator2
    Shape Primitive=::NewEnumerator4
    UseOwnerScale=true
    Yaw / Pitch / Roll=<0 0 -0.18>
   )
   &InitialMeshOrientation(
    Mesh Orientation Mode=::NewEnumerator7
    Rotation=&RandomRangeVector(Minimum=<0 0 -1>, Maximum=<0 0 1>)
    Use Rotation Vector=true
   ) {module=InitialMeshOrientation001}
   &InitialMeshOrientation(
    Mesh Orientation Mode=::NewEnumerator2
    Rotation=&RandomRangeVector(Minimum=<0.06 0 0>, Maximum=<0.06 0 0>)
    Use Rotation Vector=true
   )
   &AddVelocity(
    Velocity=&Multiply_VectorByFloat(Vector=$Particles.ShapeLocation.ShapeNormal, Float=&RandomRangeFloat(Minimum=8, Maximum=15))
    Velocity Speed Scale=&MaskFloatBySpawnGroup(Masked Float=0, Passthrough Float=1)
   )
   /SetVariables(Particles.RunesSubUV:NiagaraFloat=&RandomRangeFloat(Minimum=0, Maximum=3)) {isSetParameters=true}
  ]
  particleUpdateScript [
   &ParticleState()
   &ScaleSpriteSize(Uniform Curve Sprite Scale=DataInterface() {propertyValues {Curve {keys [
        {interpMode=RCIM_Cubic, leaveTangent=2.9844, tangentMode=RCTM_Break, time=0.0012, value=0.4443}
        {tangentMode=RCTM_User, time=0.5, value=1.1129}
        <at 0.9963 0.6467>
       ]}}})
   &ScaleMeshSize(Scale Factor=&VectorFromFloat(Value=&FloatFromCurve(FloatCurve=DataInterface() {propertyValues {Curve {
         defaultValue=0
         keys [
          {arriveTangent=4.1785, interpMode=RCIM_Cubic, leaveTangent=4.1785, tangentMode=RCTM_User, time=0, value=0}
          {arriveTangent=1.1179, interpMode=RCIM_Cubic, leaveTangent=1.1179, tangentMode=RCTM_User, time=0.2042, value=0.6228}
          {arriveTangent=0.0224, interpMode=RCIM_Cubic, leaveTangent=0.0224, tangentMode=RCTM_User, time=1, value=1}
         ]
        }}})))
   &InterpolateOverTime(Rate Of Change=5, Target Vector Value:NiagaraPosition=$Particles.Position, Variable Type=::NewEnumerator2)
   &CameraOffset()
   /SetVariables(Particles.Color2=&ColorFromCurve(ColorCurve=DataInterface() {propertyValues {
       AlphaCurve {keys [<at 0.0083 0>,<at 0.1833 0.35 RCIM_Cubic>,<at 0.5814 0.1>,<at 1 0 RCIM_Cubic>]}
       BlueCurve {keys [<at 0 0.0885 RCIM_Cubic>]}
       GreenCurve {keys [<at 0 0.0168 RCIM_Cubic>]}
       RedCurve {keys [<at 0 0.0488 RCIM_Cubic>]}
      }})) {isSetParameters=true}
   /SetVariables(Particles.Color3:LinearColor=&ColorFromCurve(ColorCurve=DataInterface() {propertyValues {
       AlphaCurve {keys [<at 0.0043 0>,<at 0.1603 0.6 RCIM_Cubic>,<at 0.5084 0.12>,<at 1 0 RCIM_Cubic>]}
       BlueCurve {keys [<at 0.4524 0.6163 RCIM_Cubic>]}
       GreenCurve {keys [<at 0.4524 0.2577 RCIM_Cubic>]}
       RedCurve {keys [<at 0.4524 0.5545 RCIM_Cubic>]}
      }})) {isSetParameters=true}
   /SetVariables(Particles.GlowScale:Vector3f=&Multiply_Vector(A=$Particles.Scale, B=<1.2 1.2 0.96>)) {isSetParameters=true}
   &ScaleColor(
    Scale Alpha=&FloatFromCurve(FloatCurve=DataInterface() {propertyValues {Curve {keys [
         {arriveTangent=4.0522, interpMode=RCIM_Cubic, leaveTangent=4.0522, tangentMode=RCTM_User, time=-0.0018, value=0.2362}
         {interpMode=RCIM_Cubic, leaveTangent=-3.9441, tangentMode=RCTM_Break, time=0.2464, value=0.9373}
         <at 1 0.0235>
        ]}}})
    Scale RGB=&MaskVectorBySpawnGroup(
     Masked Vector=<1.2 1.2 1.2>
     Passthrough Vector=&VectorFromFloat(Value=&FloatFromCurve(FloatCurve=DataInterface() {propertyValues {Curve {keys [{interpMode=RCIM_Cubic, leaveTangent=-3.6904, tangentMode=RCTM_Break, time=0, value=2}, <at 1 0>]}}}))
    )
    ScaleA=true
    ScaleRGB=true
   )
   &Drag(Drag=1.5, Use Linear Drag=true, Use Rotational Drag=true)
   &SolveForcesAndVelocity(Clamp Velocity=false, Limit Acceleration=false)
  ]
  rendererClasses [&NiagaraSpriteRendererProperties, &NiagaraMeshRendererProperties]
  renderers [
   {rendererClass=&NiagaraSpriteRendererProperties}
   {rendererClass=&NiagaraMeshRendererProperties, rendererIndex=1}
   {rendererClass=&NiagaraMeshRendererProperties, rendererIndex=2}
   {rendererClass=&NiagaraMeshRendererProperties, rendererIndex=3}
  ]
 }
}
summary {
 systemName=NS_Backpack_TallChore_Cask
 emitters [
  {
   emitterName=NE_Backpack_DrapeUltra
   rendererClasses [&NiagaraMeshRendererProperties, &NiagaraLightRendererProperties]
  }
  {emitterName=WhispFlames, rendererClasses [&NiagaraSpriteRendererProperties, &NiagaraMeshRendererProperties]}
  {emitterName=Glow, rendererClasses [&NiagaraSpriteRendererProperties]}
  {emitterName=Decal, rendererClasses [&NiagaraSpriteRendererProperties]}
 ]
 userVariables [
  User.bIsFrontEnd {type=NiagaraBool}
  User.bIsFrontEndPreview {type=NiagaraBool}
 ]
}
